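# An empty SoundCue in the sound library's weapon parent folder, named as a placeholder for weapon sounds.
# It has no root and no sound nodes, so it produces no audio.
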SoundCue /Game/Sounds/SoundLibrary/Parents/Weapon/SoundLib_Weapon_Placeholder_Cue
root=null
nodes []
